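# Niagara system for the Blue Bolt vehicle boost effect, emitting from the car's skeletal mesh sockets with painted color variants.
# Base002 burst-spawns rotating meshes and sprites in three spawn groups, mirroring spin between ports.
# Ribbons001 layers trailing ribbons with quality-scaled spawn rates and shape tweaks for frontend preview.
# S streams animated, speed-line-sized sprites at game-mode-dependent velocity; trails and sprites inherit source movement with fading influence.
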NiagaraSystem /VehicleCosmetics/Boosts/Boost_BlueBolt/FX/NS_Boost_BlueBolt_Optimized
dependencies {
 usedDataInterfaces [
  &NiagaraDataInterfaceArrayFloat3
  &NiagaraDataInterfaceCurve
  &NiagaraDataInterfaceArrayColor
  &NiagaraDataInterfaceArrayFloat2
  &NiagaraDataInterfaceVector2DCurve
  &NiagaraDataInterfaceSpriteRendererInfo
  &NiagaraDataInterfaceSkeletalMesh
 ]
 usedDynamicInputs [
  &MaskVector2dBySpawnGroup
  &Multiply_Vector2D_ByFloat
  &RandomRangeVector2D
 ]
 usedModules [
  &self:SystemSpawnScript.SystemScriptSource.SystemScriptGraph.NiagaraNodeAssignment_0.SetVariables_#1
  &SystemState
 ]
}
emitters {
 Base002 {
  emitterUpdateScript [
   &EmitterState(
    Inactive Response=::NewEnumerator1
    Life Cycle Mode=::NewEnumerator1
    Loop Behavior=::NewEnumerator1
    Loop Duration Mode=::NewEnumerator1
    UseLoopDelay=false
   )
   &VC_FilteredSocketCount()
   &SpawnRate(Use Spawn Probability=false) {enabled=false}
   &SpawnBurst_Instantaneous(Spawn Count=2, Use Spawn Probability=false, UseLoopCountLimit=false)
   &SpawnBurst_Instantaneous(Spawn Count=2, Spawn Group=1, Use Spawn Probability=false, UseLoopCountLimit=false) {module=SpawnBurst_Instantaneous001}
   &SpawnBurst_Instantaneous(Spawn Count=2, Spawn Group=2, Use Spawn Probability=false, UseLoopCountLimit=false) {module=SpawnBurst_Instantaneous002}
  ]
  particleSpawnScript [
   &VC_PaintedVariantsOverride(
    Black=<rgba 0.022 0.022 0.022 2>
    BurntSienna=<rgba 0.23 0.0642 0.0184 2>
    Cobalt=<rgba 0.08 0.08 2 2>
    Crimson=<rgba 3 0.06 0.158 2>
    DefaultColor=<rgba 0.039 0.7792 3 2>
    ForestGreen=<rgba 0.026 1.3 0.026 1>
    Grey=<rgba 0.3 0.3 0.3 2>
    Lime=<rgba 0.5 1.2 0 2>
    Orange=<rgba 2 0.7333 0 2>
    Pink=<rgba 2 0.2 1.4 2>
    Purple=<rgba 0.5 0 1.5 2>
    R1 Black:LinearColor=<rgba 0.01 0.01 0.01 1>
    R1 BurntSienna:LinearColor=<rgba 0.1 0.0295 0.01 1>
    R1 Cobalt:LinearColor=<rgba 0.02 0.0363 1 1>
    R1 Crimson:LinearColor=<rgba 0.85 0 0.0283 1>
    R1 DefaultColor=<rgba 2 0.04 1.02 1>
    R1 ForestGreen:LinearColor=<rgba 0.006 0.3 0.006 1>
    R1 Grey:LinearColor=<rgba 0.13 0.13 0.13 1>
    R1 Lime:LinearColor=<rgba 0.175 0.35 0 1>
    R1 Orange:LinearColor=<rgba 0.4 0.1467 0 1>
    R1 Pink:LinearColor=<rgba 0.6 0.15 0.45 1>
    R1 Purple:LinearColor=<rgba 0.2567 0.035 0.7 1>
    R1 Saffron:LinearColor=<rgba 0.25 0.25 0 1>
    R1 SkyBlue:LinearColor=<rgba 0.01 0.255 0.5 1>
    R1 TitaniumWhite:LinearColor=<rgba 0.3 0.3 0.3 1>
    R2 Black:LinearColor=<rgba 0.02 0.02 0.02 1>
    R2 BurntSienna:LinearColor=<rgba 0.2 0.059 0.02 1>
    R2 Cobalt:LinearColor=<rgba 0.012 0.0512 0.6 1>
    R2 Crimson:LinearColor=<rgba 0.6 0.024 0.0432 1>
    R2 DefaultColor:LinearColor=<rgba 1.5 0.15 1.05 0.4>
    R2 ForestGreen:LinearColor=<rgba 0.016 0.4 0.016 1>
    R2 Grey:LinearColor=<rgba 0.15 0.15 0.15 1>
    R2 Lime:LinearColor=<rgba 0.2333 0.4 0 1>
    R2 Orange:LinearColor=<rgba 0.4 0.1517 0.008 1>
    R2 Pink:LinearColor=<rgba 0.5 0.075 0.3583 1>
    R2 Purple:LinearColor=<rgba 0.216 0.024 0.6 1>
    R2 Saffron:LinearColor=<rgba 0.5 0.5 0.025 1>
    R2 SkyBlue:LinearColor=<rgba 0.025 0.2625 0.5 1>
    R2 TitaniumWhite:LinearColor=<rgba 0.4 0.4 0.4 1>
    Saffron=<rgba 1.3 1.3 0 2>
    SkyBlue=<rgba 0.04 1.02 2 2>
    TitaniumWhite=<rgba 1.2 1.2 1.2 2>
    Use R1 Painted Overrides=true
    Use R2 Painted Overrides=true
   )
   &InitializeParticle(
    Color=$Particles.Color
    Mesh Scale Mode=::NewEnumerator3
    Mesh Uniform Scale=&MergeHelper_FloatSwitch(0=0.13, 1=4, 2=?, 3=?, Index=&SpawnGroup(), Num=::NewEnumerator0)
    Sprite Size Mode=::NewEnumerator3
    Uniform Sprite Size=110
    UsePositionOffset=false
    Write Material Random=true
   )
   &PartitionParticles(Number Of Partitions=$Emitter.PortCount)
   &SkeletalMeshLocation(
    Bone / Socket Index=$Particles.PartitionParticles.Partition
    Bone / Socket Sampling Mode=::NewEnumerator5
    MaskSpawnGroup=false
    Mesh Sampling Type=::NewEnumerator6
    Sampled Position Offset=&AddVectorToPosition(
     Position=&MergeHelper_VectorSwitch(0=<-30 0 0>, 1=<-55 0 0>, 2=<-60 0 0>, 3=?, Index=&SpawnGroup(), Num=::NewEnumerator1)
     Vector=$User.DTCarOffset
    )
    Skeletal Mesh=$System.SkeletalMesh
   )
   &AddVelocity(Velocity=<-1500 0 0>) {enabled=false}
   &InitialMeshOrientation(Mesh Axis To Orient=<0 1 0>, Mesh Orientation Mode=::NewEnumerator2, Use Rotation Vector=false)
   /SetVariables(Particles.VisibilityTag=$Emitter.SpawnGroup) {isSetParameters=true}
  ]
  particleUpdateScript [
   &ParticleState(Kill Particles When Lifetime Has Elapsed=false)
   &UpdateMeshOrientation(
    Rotation Rate=0.2
    Rotation Vector=&MergeHelper_VectorSwitch(
     0=&SelectVectorFromArray(
      Direct Array Index=$Particles.PartitionParticles.Partition
      Vector Selection Array=DataInterface() {propertyValues {FloatData [<0 1 0>, <0 -1 0>]}}
     )
     1=&SelectVectorFromArray(
      Direct Array Index=$Particles.PartitionParticles.Partition
      Vector Selection Array=DataInterface() {propertyValues {FloatData [<0 -1 0>, <0 1 0>]}}
     )
     2=?
     3=?
     Index=&SpawnGroup()
     Num=::NewEnumerator0
    )
   )
   &SolveForcesAndVelocity(Clamp Velocity=false, Limit Acceleration=false)
  ]
  rendererClasses [&NiagaraMeshRendererProperties, &NiagaraSpriteRendererProperties]
  renderers [
   {rendererClass=&NiagaraMeshRendererProperties, rendererIndex=0}
   {rendererClass=&NiagaraMeshRendererProperties, rendererIndex=1}
   {rendererClass=&NiagaraSpriteRendererProperties, rendererIndex=2}
  ]
 }
 Ribbons001 {
  emitterUpdateScript [
   &EmitterState()
   &VC_FilteredSocketCount()
   &VC_RateQualityHandler(High Spawn Rate=50, Low Spawn Rate=30, Medium Spawn Rate=40, Use Custom Rate?=true)
   &VC_SpawnRibbon_v1(Number of Ribbons=$Emitter.PortCount, SpawnRate=$Emitter.Spawn Rate)
  ]
  particleSpawnScript [
   &VC_PaintedVariantsOverride(
    Black=<rgba 0.3 0.3 0.3 1>
    BurntSienna=<rgba 0.7 0.3436 0.245 1>
    Cobalt=<rgba 0.56 0.588 1.4 1>
    Crimson=<rgba 1.4 0.42 0.4527 1>
    DefaultColor=<rgba 3.5 2.3333 2.9805 0.287>
    ForestGreen=<rgba 0.3 1.2 0.3 1>
    Grey=<rgba 0.7 0.7 0.7 1>
    Lime=<rgba 0.832 1.2 0.096 1>
    Orange=<rgba 1.5 0.55 0 1>
    Pink=<rgba 1.5 0.6 1.425 1>
    Purple=<rgba 0.9173 0.32 1.6 1>
    R1 Black:LinearColor=<rgba 0.006 0.006 0.006 1>
    R1 BurntSienna:LinearColor=<rgba 0.02 0.0106 0.008 1>
    R1 Cobalt:LinearColor=<rgba 0.01 0.016 0.1 1>
    R1 Crimson:LinearColor=<rgba 0.06 0.003 0.0045 1>
    R1 DefaultColor=<rgba 0 1 2 1>
    R1 ForestGreen:LinearColor=<rgba 0 0.05 0 1>
    R1 Grey:LinearColor=<rgba 0.03 0.03 0.03 1>
    R1 Lime:LinearColor=<rgba 0.0267 0.04 0 1>
    R1 Orange:LinearColor=<rgba 0.06 0.022 0 1>
    R1 Pink:LinearColor=<rgba 0.08 0.032 0.076 1>
    R1 Purple:LinearColor=<rgba 0.0366 0.0056 0.08 1>
    R1 Saffron:LinearColor=<rgba 0.07 0.07 0.014 1>
    R1 SkyBlue:LinearColor=<rgba 0.002 0.0428 0.1 1>
    R1 TitaniumWhite:LinearColor=<rgba 0.08 0.08 0.08 1>
    R2 Black:LinearColor=<rgba 0.005 0.005 0.005 1>
    R2 BurntSienna:LinearColor=<rgba 0.045 0.0238 0.018 1>
    R2 Cobalt:LinearColor=<rgba 0.015 0.024 0.15 1>
    R2 Crimson:LinearColor=<rgba 0.1 0.004 0.0066 1>
    R2 DefaultColor:LinearColor=<rgba 0.0221 0.0221 0.4427 1>
    R2 ForestGreen:LinearColor=<rgba 0 0.06 0 1>
    R2 Grey:LinearColor=<rgba 0.04 0.04 0.04 1>
    R2 Lime:LinearColor=<rgba 0.0333 0.05 0 1>
    R2 Orange:LinearColor=<rgba 0.08 0.0293 0 1>
    R2 Pink:LinearColor=<rgba 0.09 0.036 0.0855 1>
    R2 Purple:LinearColor=<rgba 0.0458 0.007 0.1 1>
    R2 Saffron:LinearColor=<rgba 0.08 0.08 0.016 1>
    R2 SkyBlue:LinearColor=<rgba 0.0024 0.0514 0.12 1>
    R2 TitaniumWhite:LinearColor=<rgba 0.1 0.1 0.1 1>
    R3 Black:LinearColor=<rgba 0.005 0.005 0.005 1>
    R3 BurntSienna:LinearColor=<rgba 0.005 0.005 0.005 1>
    R3 Cobalt:LinearColor=<rgba 0.005 0.005 0.005 1>
    R3 Crimson:LinearColor=<rgba 0.005 0.005 0.005 1>
    R3 DefaultColor:LinearColor=<rgba 0.0221 0.0221 0.4427 1>
    R3 ForestGreen:LinearColor=<rgba 0.005 0.005 0.005 1>
    R3 Gold:LinearColor=<rgba 0.005 0.005 0.005 1>
    R3 Grey:LinearColor=<rgba 0.005 0.005 0.005 1>
    R3 Lime:LinearColor=<rgba 0.005 0.005 0.005 1>
    R3 Orange:LinearColor=<rgba 0.005 0.005 0.005 1>
    R3 Pink:LinearColor=<rgba 0.005 0.005 0.005 1>
    R3 Purple:LinearColor=<rgba 0.005 0.005 0.005 1>
    R3 Saffron:LinearColor=<rgba 0.005 0.005 0.005 1>
    R3 SkyBlue:LinearColor=<rgba 0.005 0.005 0.005 1>
    R3 TitaniumWhite:LinearColor=<rgba 0.005 0.005 0.005 1>
    Saffron=<rgba 1.3 1.3 0.13 1>
    SkyBlue=<rgba 0.6 1.1833 2 1>
    TitaniumWhite=<rgba 1.3 1.3 1.3 1>
    Use R1 Painted Overrides=true
    Use R2 Painted Overrides=true
    Use R3 Painted Overrides=true
   )
   &VC_InitializeRibbon_v1(
    Color=$Particles.Color
    Custom Offset=true
    Enable Skeletal Position=true
    Lifetime=&MakeCustomFloatFromBool(Bool=$User.bIsFrontendPreview, True Float=0.7, False Float=0.6)
    Local Vector Offset=&AddVectorToPosition(Position=<-55 0 0>, Vector=$User.DTCarOffset)
    Offset Amount=1
    Ribbon Width=80
   )
   &AddVelocity(Velocity=<-2000 0 0>) {module=AddVelocity001}
  ]
  particleUpdateScript [
   &ParticleState()
   &InheritSourceMovement(
    Applied Position Scale=&VectorFromFloat(Value=&FloatFromCurve(FloatCurve=DataInterface() {propertyValues {Curve {keys [<at 0 1 RCIM_Cubic>, {tangentMode=RCTM_User, time=1, value=0}]}}}))
    ApplyForce=false
    ApplyPosition=true
    ApplyPositionFalloff=false
   )
   &SolveForcesAndVelocity(Clamp Velocity=false, Limit Acceleration=false)
   &ScaleColor(
    Scale Alpha=&FloatFromCurve(FloatCurve=DataInterface() {propertyValues {Curve {keys [
         {arriveTangent=15.4706, interpMode=RCIM_Cubic, leaveTangent=15.4706, tangentMode=RCTM_User, time=0, value=0}
         {arriveTangentWeight=0.0786, interpMode=RCIM_Cubic, leaveTangentWeight=0.2359, tangentMode=RCTM_User, time=0.1, value=1}
         {
          arriveTangent=-0.7912
          arriveTangentWeight=0.3037
          interpMode=RCIM_Cubic
          leaveTangent=-0.7912
          tangentMode=RCTM_User
          time=0.3
          value=0.0344
         }
        ]}}})
    Scale RGB=&VectorFromFloat(Value=&FloatFromCurve(FloatCurve=DataInterface() {propertyValues {Curve {keys [{interpMode=RCIM_Cubic, tangentMode=RCTM_Break, time=0, value=2.5}, {interpMode=RCIM_Cubic, tangentMode=RCTM_Break, time=0.025, value=1}]}}}))
    ScaleA=true
    ScaleRGB=true
   )
   &ScaleRibbonWidth(Ribbon Width Scale=&MakeCustomFloatFromBool(
     Bool=$User.bIsFrontendPreview
     True Float=&FloatFromCurve(FloatCurve=DataInterface() {propertyValues {Curve {keys [
          {interpMode=RCIM_Cubic, tangentMode=RCTM_Break, time=0.1, value=1}
          {interpMode=RCIM_Cubic, tangentMode=RCTM_Break, time=0.3, value=0.7}
          {interpMode=RCIM_Cubic, tangentMode=RCTM_Break, time=1, value=0}
         ]}}})
     False Float=&FloatFromCurve(FloatCurve=DataInterface() {propertyValues {Curve {keys [<at 0 1 RCIM_Cubic>, {arriveTangent=-4.3174, tangentMode=RCTM_User, time=0.4, value=0}]}}})
    ))
   &VC_UpdateRibbon_v1(Enable Skeletal Position=true)
   /SetVariables(
    Particles.Ribbon2:LinearColor=&VC_DynamicScaleColor(
     Color to Scale:LinearColor=$Particles.R2 Color
     RGB Scale:Vector3f=&VectorFromFloat(Value=&MakeCustomFloatFromBool(
       Bool=$User.bIsPainted
       True Float=&FloatFromCurve(FloatCurve=DataInterface() {propertyValues {Curve {keys [<at 0 3 RCIM_Cubic>,<at 0.5 2 RCIM_Cubic>]}}})
       False Float=&FloatFromCurve(FloatCurve=DataInterface() {propertyValues {Curve {keys [
            <at 0 4 RCIM_Cubic>
            {arriveTangent=-7.6, arriveTangentWeight=0.2555, interpMode=RCIM_Cubic, leaveTangent=-7.6, leaveTangentWeight=1.0221, time=0.1, value=2}
            <at 0.5 0.2 RCIM_Cubic>
           ]}}})
      ))
     Alpha Scale:NiagaraFloat=&FloatFromCurve(
      FloatCurve=DataInterface() {propertyValues {Curve {keys [
          <at 0 0 RCIM_Cubic>
          {arriveTangent=0.7143, interpMode=RCIM_Cubic, leaveTangent=0.7143, time=0.1, value=1}
          {arriveTangent=-1.4286, interpMode=RCIM_Cubic, leaveTangent=-1.4286, time=0.7, value=0.5}
          <at 0.8 0 RCIM_Cubic>
         ]}}}
      Scale Curve=1.4
     )
    )
    Particles.Ribbon2Width=75
    Particles.Ribbon3:LinearColor=&VC_DynamicScaleColor(
     Color to Scale:LinearColor=$Particles.R3 Color
     RGB Scale:Vector3f=&VectorFromFloat(Value=&MakeCustomFloatFromBool(
       Bool=$User.bIsPainted
       True Float=&FloatFromCurve(FloatCurve=DataInterface() {propertyValues {Curve {keys [<at 0 3 RCIM_Cubic>,<at 0.5 2 RCIM_Cubic>]}}})
       False Float=&FloatFromCurve(FloatCurve=DataInterface() {propertyValues {Curve {keys [
            <at 0 4 RCIM_Cubic>
            {arriveTangent=-7.6, arriveTangentWeight=0.2555, interpMode=RCIM_Cubic, leaveTangent=-7.6, leaveTangentWeight=1.0221, time=0.1, value=2}
            <at 0.5 0.2 RCIM_Cubic>
           ]}}})
      ))
     Alpha Scale:NiagaraFloat=&FloatFromCurve(
      FloatCurve=DataInterface() {propertyValues {Curve {keys [
          <at 0 0 RCIM_Cubic>
          {arriveTangent=0.7143, interpMode=RCIM_Cubic, leaveTangent=0.7143, time=0.1, value=1}
          {arriveTangent=-1.4286, interpMode=RCIM_Cubic, leaveTangent=-1.4286, time=0.7, value=0.5}
          <at 0.8 0 RCIM_Cubic>
         ]}}}
      Scale Curve=1.4
     )
    )
   ) {isSetParameters=true}
  ]
  rendererClasses [&NiagaraRibbonRendererProperties]
  renderers [
   {rendererClass=&NiagaraRibbonRendererProperties}
   {rendererClass=&NiagaraRibbonRendererProperties, rendererIndex=1}
   {rendererClass=&NiagaraRibbonRendererProperties, rendererIndex=2}
   {rendererClass=&NiagaraRibbonRendererProperties, rendererIndex=3}
  ]
 }
 S {
  emitterUpdateScript [
   &EmitterState()
   &VC_FilteredSocketCount()
   &SpawnRate(SpawnRate=70, Use Spawn Probability=false)
   &SpawnRate(SpawnGroup=1, SpawnRate=50, Use Spawn Probability=false) {module=SpawnRate001}
  ]
  particleSpawnScript [
   &VC_PaintedVariantsOverride(
    Black=<rgba 0.06 0.06 0.06 1>
    BurntSienna=<rgba 0.23 0.0679 0.023 1>
    Cobalt=<rgba 0.135 0.1605 0.9 1>
    Crimson=<rgba 0.95 0.095 0.1235 1>
    DefaultColor=&SelectLinearColorFromArray(
     Array Sampling Mode=::NewEnumerator0
     Direct Array Index=?
     Color Selection Array=DataInterface() {propertyValues {ColorData [<rgba 1 0.2874 1 1>, <rgba 0.2195 1 1 1>]}}
     Random Seed=0
    )
    ForestGreen=<rgba 0.085 0.85 0.085 1>
    Grey=<rgba 0.25 0.25 0.25 1>
    Lime=<rgba 0.35 0.6 0 1>
    Orange=<rgba 0.95 0.3784 0.0475 1>
    Pink=<rgba 0.85 0.255 0.8004 1>
    Purple=<rgba 0.4538 0.135 0.9 1>
    R1 Black:LinearColor=<rgba 0.15 0.15 0.15 1>
    R1 BurntSienna:LinearColor=<rgba 0.3 0.0885 0.03 1>
    R1 Cobalt:LinearColor=<rgba 0.19 0.2153 0.95 1>
    R1 Crimson:LinearColor=<rgba 0.95 0.1425 0.1694 1>
    R1 DefaultColor=<rgba 0.5677 0.7839 1 1>
    R1 ForestGreen:LinearColor=<rgba 0.095 0.95 0.095 1>
    R1 Grey:LinearColor=<rgba 0.3 0.3 0.3 1>
    R1 Lime:LinearColor=<rgba 0.3792 0.65 0 1>
    R1 Orange:LinearColor=<rgba 0.95 0.4085 0.095 1>
    R1 Pink:LinearColor=<rgba 0.9 0.315 0.8512 1>
    R1 Purple:LinearColor=<rgba 0.5067 0.19 0.95 1>
    R1 Saffron:LinearColor=<rgba 0.9 0.9 0.135 1>
    R1 SkyBlue:LinearColor=<rgba 0.1425 0.5462 0.95 1>
    R1 TitaniumWhite:LinearColor=<rgba 0.85 0.85 0.85 1>
    Saffron=<rgba 0.9 0.9 0.135 1>
    SkyBlue=<rgba 0.135 0.5175 0.9 1>
    TitaniumWhite=<rgba 0.8 0.8 0.8 1>
    Use R1 Painted Overrides=true
   )
   &InitializeParticle(
    Color=$Particles.Color
    Lifetime=&RandomRangeFloat(Minimum=0.3, Maximum=0.5)
    Sprite Rotation Mode=::NewEnumerator2
    Sprite Size Mode=::NewEnumerator3
    Uniform Sprite Size=&MergeHelper_FloatSwitch(
     0=&RandomRangeFloat(Minimum=10, Maximum=15)
     1=5
     2=?
     3=?
     Index=&SpawnGroup()
     Num=::NewEnumerator0
    )
    UsePositionOffset=false
    Write Material Random=true
   )
   &PartitionParticles(Number Of Partitions=$Emitter.PortCount)
   &SkeletalMeshLocation(
    Bone / Socket Index=$Particles.PartitionParticles.Partition
    Bone / Socket Sampling Mode=::NewEnumerator5
    MaskSpawnGroup=false
    Mesh Sampling Type=::NewEnumerator6
    Skeletal Mesh=$System.SkeletalMesh
   )
   &AddVelocity(Velocity=&SelectVectorFromArray(
     Direct Array Index=$User.GameModeID
     Vector Selection Array=DataInterface() {propertyValues {FloatData [<-600 0 0>, <-500 0 0>, <-400 0 0>]}}
    ))
   &ShapeLocation(
    MaskSpawnGroup=false
    Offset=&AddVectorToPosition(Position=<-90 0 0>, Vector=$User.DTCarOffset)
    Offset Coordinate Space=::Local
    Offset Mode=::NewEnumerator0
    Override Seed=false
    Sphere Radius=&MergeHelper_FloatSwitch(0=25, 1=29, 2=?, 3=?, Index=&SpawnGroup(), Num=::NewEnumerator0)
    UseOwnerScale=true
   )
   /SetVariables(Particles.VisibilityTag=$Emitter.SpawnGroup) {isSetParameters=true}
  ]
  particleUpdateScript [
   &ParticleState()
   &SpriteRotationRate(Rotation Rate=&MergeHelper_FloatSwitch(
     0=&RandomRangeFloat(Minimum=-200, Maximum=-400)
     1=0
     2=?
     3=?
     Index=&SpawnGroup()
     Num=::NewEnumerator0
    ))
   &InheritSourceMovement(
    Applied Position Scale=&VectorFromFloat(Value=&FloatFromCurve(FloatCurve=DataInterface() {propertyValues {Curve {keys [<at 0 1 RCIM_Cubic>,<at 1 0 RCIM_Cubic>]}}}))
    ApplyForce=false
    ApplyPosition=true
    ApplyPositionFalloff=false
   )
   &ScaleSpriteSize(
    Scale Sprite Size Mode=::NewEnumerator0
    Uniform Scale Factor=&MergeHelper_FloatSwitch(
     0=&FloatFromCurve(FloatCurve=DataInterface() {propertyValues {Curve {keys [<at 0 1>,<at 1 0.5>]}}})
     1=1
     2=?
     3=?
     Index=&SpawnGroup()
     Num=::NewEnumerator0
    )
   )
   /SetVariables(Particles.speedlinessize:Vector2f=&Multiply_Vector2D(
     A=$Particles.SpriteSize
     B=&Vector2DFromCurve(
      Vector2Curve=DataInterface() {propertyValues {XCurve {keys [<at 0 1>,<at 1 1>]}, YCurve {keys [<at 0.5 1>,<at 1 0.2>]}}}
      ScaleCurve=&SelectVector2DFromArray(
       Array Sampling Mode=::NewEnumerator0
       Direct Array Index=?
       Vector 2D Selection Array=DataInterface() {propertyValues {FloatData [<0.4 5>, <0.4 15>]}}
       Random Seed=0
      )
     )
    )) {isSetParameters=true}
   &SolveForcesAndVelocity(Clamp Velocity=false, Limit Acceleration=false)
   &SubUVAnimation(
    Evaluation Type=::NewEnumerator0
    "Sprite Renderer "=DataInterface() {propertyValues {SpriteRenderer=&self:NiagaraEmitter_2.NiagaraSpriteRendererProperties_2}}
    SubUV Animation Mode:ENiagara_SubUVLookupModeV2=::NewEnumerator2
    UseEndFrame=false
    UseStartFrame=false
   )
   &ScaleColor(
    Scale Alpha=&FloatFromCurve(FloatCurve=DataInterface() {propertyValues {Curve {
        defaultValue=1
        keys [
         {interpMode=RCIM_Cubic, tangentMode=RCTM_User, time=0.5, value=1}
         {arriveTangentWeight=0.1817, interpMode=RCIM_Cubic, tangentMode=RCTM_User, time=1, value=0}
        ]
       }}})
    Scale RGB=&VectorFromFloat(Value=8)
    ScaleA=true
    ScaleRGB=true
   )
  ]
  rendererClasses [&NiagaraSpriteRendererProperties]
  renderers [
   {rendererClass=&NiagaraSpriteRendererProperties}
   {rendererClass=&NiagaraSpriteRendererProperties, rendererIndex=1}
  ]
 }
}
summary {
 systemName=NS_Boost_BlueBolt_Optimized
 emitters [
  {emitterName=Base002, rendererClasses [&NiagaraMeshRendererProperties, &NiagaraSpriteRendererProperties]}
  {emitterName=Ribbons001, rendererClasses [&NiagaraRibbonRendererProperties]}
  {emitterName=S, rendererClasses [&NiagaraSpriteRendererProperties]}
 ]
 userVariables [
  User.bIsFrontendPreview {type=NiagaraBool}
  User.bIsPainted {type=NiagaraBool}
  User.ColorVariantID {type=NiagaraInt32}
  User.DTCarOffset {type=Vector3f, description="", defaultValue {struct=Vector3f, value=<0 0 0>}}
  User.GameModeID {type=NiagaraInt32}
  User.PaintedColor {type=LinearColor, description="", defaultValue {struct=LinearColor, value=<rgba 1 1 1 1>}}
  User.PaintedScalar {type=NiagaraFloat, defaultValue=1, description=""}
  User.PortCount {type=NiagaraInt32, defaultValue=2}
 ]
}
