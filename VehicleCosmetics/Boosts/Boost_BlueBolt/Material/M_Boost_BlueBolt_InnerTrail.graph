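# Translucent unlit material for the inner trail of the Blue Bolt vehicle boost, built for Niagara ribbons.
# Panning smoke, faded at both ribbon ends, is layered over a glow gradient to form emissive color and a depth-faded opacity.
# Dynamic parameters scale pan speed, blend in tinting, adjust smoke brightness, and swap default pink smoke and blue glow tints for particle color via IsPainted.
# Emissive color passes through the shared MF_FX_EmissiveAdjustments function.
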Material /VehicleCosmetics/Boosts/Boost_BlueBolt/Material/M_Boost_BlueBolt_InnerTrail
graph {
 properties {
  bAutomaticallySetUsageInEditor=false
  bUsedWithNiagaraRibbons=true
  blendMode=BLEND_Translucent
  shadingModel=MSM_Unlit
  usedShadingModels=Unlit
 }
 nodes [
  Constant3Vector() {constant=<rgba 1 0.1 0.6 0>}
  n1 = DynamicParameter() {defaultValue=<rgba 1 1 1 0>, paramNames [time,Param2,Param3,IsPainted]}
  n2 = Time()
  n3 = TextureSample(UVs=Panner(Coordinate=TextureCoordinate(), Time=Multiply(A=@n1.time, B=@n2)) {speedX=-1.5}) {texture=/DealershipSharedLibrary/FX/Textures/TexturePackage/T_VCVFX_SmokeTrail_MoreDetailed}
  n4 = TextureCoordinate()
  MaterialExpressionConstantClamp_17 = Saturate(None=Multiply(A=ComponentMask(None=@n4) {r=true}, B=Constant() {r=20}))
  n5 = Power(Base=@MaterialExpressionConstantClamp_17, Exp=Constant() {r=2})
  n6 = ComponentMask(None=TextureCoordinate()) {r=true}
  MaterialExpressionConstantClamp_19 = Saturate(None=Multiply(A=OneMinus(None=@n6), B=Constant() {r=1.5}))
  n7 = Constant() {r=20}
  n8 = Multiply(A=Multiply(A=Add(A=@n3.R, B=@n3.R), B=@n3.G), B=Multiply(A=@n5, B=Power(Base=@MaterialExpressionConstantClamp_19, Exp=@n7)))
  MaterialExpressionVertexColor_6 = ParticleColor()
  n9 = DynamicParameter() {defaultValue=<rgba 1 1 1 0>, paramNames [time,Param2,Param3,IsPainted]}
  n10 = LinearInterpolate(
   A=Multiply(
    A=VectorParameter() {parameterName=InnterTrailPink, defaultValue=<rgba 1 0.1 0.6 1>}.RGB
    B=ScalarParameter() {defaultValue=1, parameterName=BrughtPink}
   )
   B=@MaterialExpressionVertexColor_6.RGB
   Alpha=@n9.IsPainted
  )
  n11 = LinearInterpolate(A=@n8, B=Multiply(A=Constant() {r=0.7}, B=Multiply(A=@n10, B=@n8)), Alpha=@n9.Param2)
  n12 = LinearInterpolate(A=Multiply(A=Constant() {r=1.2}, B=Multiply(A=@MaterialExpressionVertexColor_6.RGB, B=@n11)), B=@n11, Alpha=@n9.Param2)
  n13 = Multiply(A=Constant() {r=1.8}, B=@n12)
  n14 = TextureSample(UVs=TextureCoordinate()) {samplerType=SAMPLERTYPE_LinearColor, texture=/DealershipSharedLibrary/FX/Textures/Gradient_Masks/T_VCVFX_GlowTrail}
  n15 = Constant3Vector() {constant=<rgba 0.25 0.7 1 0>}
  n16 = LinearInterpolate(A=@n14.R, B=Multiply(A=LinearInterpolate(A=@n15, B=@MaterialExpressionVertexColor_6.RGB, Alpha=@n9.IsPainted), B=@n14.R), Alpha=@n9.Param2)
  MaterialExpressionConstantClamp_15 = Saturate(None=Add(A=LinearInterpolate(A=@n13, B=@n12, Alpha=@n9.Param3), B=Multiply(A=@MaterialExpressionVertexColor_6.RGB, B=@n16)))
  MaterialFunctionCall(Emissive=@MaterialExpressionConstantClamp_15) {materialFunction=&MF_FX_EmissiveAdjustments}
  n17 = ComponentMask(None=TextureCoordinate()) {r=true}
  MaterialExpressionConstantClamp_18 = Saturate(None=Multiply(A=OneMinus(None=@n17), B=Constant() {r=1}))
  n18 = Multiply(A=@n5, B=Power(Base=@MaterialExpressionConstantClamp_18, Exp=Constant() {r=4}))
  n19 = ParticleColor()
  MaterialExpressionConstantClamp_16 = Saturate(None=Add(A=Multiply(A=@n14.R, B=@n18), B=Multiply(A=@n8, B=@n19.A)))
  Multiply(A=@MaterialExpressionConstantClamp_16, B=DepthFade())
 ]
}
